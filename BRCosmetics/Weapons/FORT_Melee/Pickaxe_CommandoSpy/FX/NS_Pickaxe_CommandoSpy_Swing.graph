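# Niagara swing trail effect for the CommandoSpy pickaxe, built from two short-lived animation-trail ribbon emitters.
# Trail_Main is a warm yellow-to-orange streak that fades out by mid-life, while Trail001 is a near-black trail that fades later.
# Both multiply spawn alpha by System.TrailFade and tint RGB by System.TOD.
# Time-of-day scalars also adjust each trail's alpha, with Trail_Main adding a curve indexed by normalized time of day.
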NiagaraSystem /BRCosmetics/Weapons/FORT_Melee/Pickaxe_CommandoSpy/FX/NS_Pickaxe_CommandoSpy_Swing
dependencies {
 usedDataInterfaces [&NiagaraDataInterfaceColorCurve, &NiagaraDataInterfaceCurve, &NiagaraDataInterfaceSocketReader]
 usedDynamicInputs []
 usedModules [
  &self:SystemSpawnScript.SystemScriptSource.SystemScriptGraph.NiagaraNodeAssignment_3.SetVariables_#1
  &self:SystemSpawnScript.SystemScriptSource.SystemScriptGraph.NiagaraNodeAssignment_0.SetVariables_#2
  &SystemState
  &self:SystemSpawnScript.SystemScriptSource.SystemScriptGraph.NiagaraNodeAssignment_1.SetVariables_#3
 ]
}
emitters {
 Trail001 {
  emitterUpdateScript [
   &EmitterState()
   &FN_SpawnAnimTrail(AnimTrailWidthMode=::FromSecond)
  ]
  particleSpawnScript [
   &FN_InitializeAnimTrailBoneIndecies() {module=FN_InitializeAnimTrailBoneIndecies001}
   &FN_InitializeAnimTrail(
    Color=&MakeLinearColor(
     R:NiagaraFloat=1
     G:NiagaraFloat=1
     B:NiagaraFloat=1
     A:NiagaraFloat=&Multiply_Float(A=0.5, B=$System.TrailFade)
    )
    Lifetime=0.33
   )
  ]
  particleUpdateScript [
   &ScaleColor(
    Scale Alpha=&FloatFromCurve(
     FloatCurve=DataInterface() {propertyValues {Curve {keys [{arriveTangent=0.0074, interpMode=RCIM_Cubic, leaveTangent=0.0074, tangentMode=RCTM_User, time=0.5, value=1}, <at 1 0 RCIM_Cubic>]}}}
     Scale Curve=&FNTimeOfDayScalar(Night Scale=0.8, Morning Scale=0.6, Day Scale=2.3, Evening Scale=1)
    )
    Scale RGB=&Multiply_Vector(
     A=&MakeVectorFromLinearColorRGB(Color=&ColorFromCurve(ColorCurve=DataInterface() {propertyValues {AlphaCurve {keys []}, BlueCurve {keys [<at 0.0523 0.03>]}, GreenCurve {keys [<at 0.0523 0.0255>]}, RedCurve {keys [<at 0.0523 0.0248>]}}}))
     B=&MakeVectorFromLinearColorRGB(Color=$System.TOD)
    )
    ScaleA=true
    ScaleRGB=true
   ) {module=ScaleColor001}
   &FN_UpdateAnimTrail()
  ]
  rendererClasses [&NiagaraRibbonRendererProperties]
  renderers [{rendererClass=&NiagaraRibbonRendererProperties}]
 }
 Trail_Main {
  emitterUpdateScript [
   &EmitterState()
   &FN_SpawnAnimTrail(AnimTrailWidthMode=::FromSecond)
  ]
  particleSpawnScript [
   &FN_InitializeAnimTrailBoneIndecies() {module=FN_InitializeAnimTrailBoneIndecies001}
   &FN_InitializeAnimTrail(
    Color=&MakeLinearColor(
     R:NiagaraFloat=1
     G:NiagaraFloat=1
     B:NiagaraFloat=1
     A:NiagaraFloat=&Multiply_Float(A=0.2, B=$System.TrailFade)
    )
    Lifetime=0.33
   )
  ]
  particleUpdateScript [
   &ScaleColor(
    Scale Alpha=&FloatFromCurve(
     FloatCurve=DataInterface() {propertyValues {Curve {keys [{arriveTangent=0.1961, interpMode=RCIM_Cubic, leaveTangent=0.1961, tangentMode=RCTM_User, time=0.2, value=1}, <at 0.5 0 RCIM_Cubic>]}}}
     Scale Curve=&FloatFromCurve(
      FloatCurve=DataInterface() {propertyValues {Curve {keys [<at 0 0.3>,<at 0.3 1>,<at 0.6 1>,<at 1 0.3>]}}}
      CurveIndex=Linked() {linkedVariable {name=NPC.FortniteNPC.TimeOfDay_0-1, type=NiagaraFloat}}
      Scale Curve=&FNTimeOfDayScalar(Night Scale=3, Morning Scale=1, Day Scale=0.8, Evening Scale=5)
     )
    )
    Scale RGB=&Multiply_Vector(
     A=&MakeVectorFromLinearColorRGB(Color=&ColorFromCurve(ColorCurve=DataInterface() {propertyValues {
         AlphaCurve {keys []}
         BlueCurve {keys [<at 0 0.2124>,<at 0.2068 0.05>]}
         GreenCurve {keys [<at 0 0.6873>,<at 0.2068 0.1733>]}
         RedCurve {keys [<at 0 0.9948>,<at 0.2068 0.6406>]}
        }}))
     B=&MakeVectorFromLinearColorRGB(Color=$System.TOD)
    )
    ScaleA=true
    ScaleRGB=true
   )
   &FN_UpdateAnimTrail()
  ]
  rendererClasses [&NiagaraRibbonRendererProperties]
  renderers [{rendererClass=&NiagaraRibbonRendererProperties}]
 }
}
summary {
 systemName=NS_Pickaxe_CommandoSpy_Swing
 emitters [
  {emitterName=Trail_Main, rendererClasses [&NiagaraRibbonRendererProperties]}
  {emitterName=Trail001, rendererClasses [&NiagaraRibbonRendererProperties]}
 ]
 userVariables [
  User.bIsFrontEndPreview {type=NiagaraBool}
  User.RibbonIDOffset {type=NiagaraInt32}
 ]
}
